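# SSI_M_Debug is a debug material for the snow and sand interaction gameplay cue effects.
# It is an unlit, masked material enabled for Niagara sprites.
# The graph holds a sample of the SSI_T_Version texture and a color vector parameter C that defaults to red, with no connections shown between them.
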Material /FNGameplayCues/Effects/SnowSandInteraction/Materials/SSI_M_Debug
graph {
 properties {bUsedWithNiagaraSprites=true, blendMode=BLEND_Masked, pixelDepthOffsetMode=PDOM_Legacy, shadingModel=MSM_Unlit, usedShadingModels=Unlit}
 nodes [
  TextureSample() {texture=/FNGameplayCues/Effects/SnowSandInteraction/Textures/SSI_T_Version}
  VectorParameter() {parameterName=C, defaultValue=<rgba 1 0 0 1>}
 ]
}
