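# Masked, two-sided subsurface material for Niagara sand sprites in the snow/sand interaction effects.
# Panning cloud noise, offset by a dynamic parameter, is sharpened by power and contrast, then multiplied by a V-flipped soft flame mask and particle alpha to feed temporal AA dithering.
# The noise also blends a color tinted by particle color.
# A negated vertex-normal dot atmospheric-light term scales a dynamic-parameter tint for away-from-light shading.
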Material /FNGameplayCues/Effects/SnowSandInteraction/Materials/SSI_M_Sand
graph {
 properties {
  bUsedWithNiagaraSprites=true
  blendMode=BLEND_Masked
  pixelDepthOffsetMode=PDOM_Legacy
  shadingModel=MSM_Subsurface
  twoSided=true
  usedShadingModels=Subsurface
 }
 nodes [
  Constant() {r=1}
  Constant() {r=0.1}
  n1 = ParticleColor()
  n2 = Multiply(A=TextureCoordinate(), B=ScalarParameter() {defaultValue=0.1, parameterName=te})
  n3 = DynamicParameter()
  n4 = Time() {bOverride_Period=true, period=600}
  n5 = VectorParameter()
  n6 = AppendVector(A=@n5.R, B=@n5.R)
  n7 = TextureSample(UVs=Add(A=Add(A=@n2, B=@n3.Param1), B=Multiply(A=@n4, B=@n6))) {texture=/Game/Packages/Fortress_Transylvania/Haunted_Forest/Swamps/Textures/clouds_dust_detail_01}
  n8 = Power(Base=@n7.R, Exp=ScalarParameter() {defaultValue=1.5, parameterName=W})
  n9 = ScalarParameter() {defaultValue=1.1, parameterName=S}
  n10 = MaterialFunctionCall(Float2=TextureCoordinate()) {materialFunction=&BreakOutFloat2Components}
  n11 = OneMinus(None=@n10.G)
  n12 = Multiply(
   A=MaterialFunctionCall(In=Multiply(A=@n8, B=@n9), Contrast=ScalarParameter() {defaultValue=5, parameterName=Wc}) {materialFunction=&CheapContrast}.Result
   B=TextureSample(UVs=MaterialFunctionCall(X=@n10.R, Y=@n11) {materialFunction=&MakeFloat2}.Result) {texture=/Game/Characters/Player/Female/Medium/Bodies/F_MED_Dark_Elf/FX/T_Flame_mask_soft}.R
  )
  MaterialFunctionCall(Alpha Threshold=Saturate(None=Multiply(A=@n1.A, B=@n12))) {materialFunction=&DitherTemporalAA_Fortnite}
  n13 = DynamicParameter()
  n14 = Multiply(A=DotProduct(A=VertexNormalWS(), B=AtmosphericLightVector())) {constB=-1}
  Multiply(
   A=AppendVector(A=AppendVector(A=@n13.Param2, B=@n13.Param2), B=@n13.Param4)
   B=Multiply(A=Saturate(None=@n14), B=ScalarParameter() {defaultValue=0.7, parameterName=sTestBSSS}) {constB=0.6}
  )
  Reroute(None=Multiply(A=LinearInterpolate(Alpha=@n7.R), B=@n1.RGB))
  ScalarParameter() {defaultValue=0.1}
 ]
}
