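# Subsurface-shaded snowball material for Niagara mesh particles in the snow and sand interaction effects.
# Soft Perlin noise, sampled by instance-space XY and shifted by a Random dynamic parameter, scales the vertex normal into a lumpy offset driven by WPOStr and scale.
# ColorSSS times dynamic parameter D feeds the shared emissive adjustment function.
# A base Color parameter and tiled wave normal map are also present; material output wiring is not shown.
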Material /FNGameplayCues/Effects/SnowSandInteraction/Materials/SSI_M_Snowball
graph {
 properties {bUsedWithNiagaraMeshParticles=true, pixelDepthOffsetMode=PDOM_Legacy, shadingModel=MSM_Subsurface, usedShadingModels=Subsurface}
 nodes [
  Constant()
  MaterialFunctionCall() {materialFunction=&LocalPosition}
  n1 = DynamicParameter() {paramNames [Random,scale,WPOStr,D]}
  MaterialFunctionCall(Emissive=Multiply(A=VectorParameter() {parameterName=ColorSSS, defaultValue=<rgba 0.3334 0.3556 0.37 0>}.RGB, B=@n1.D)) {materialFunction=&MF_FX_EmissiveAdjustments}
  n2 = ComponentMask(None=TransformPosition(None=WorldPosition().XYZ) {transformSourceType=TRANSFORMPOSSOURCE_World, transformType=TRANSFORMPOSSOURCE_Instance}) {g=true, r=true}
  n3 = TextureSample(UVs=Add(A=Multiply(A=@n2) {constB=0.01}, B=@n1.Random) {constB=0.2}) {samplerType=SAMPLERTYPE_Masks, texture=/Game/Characters/Player/Female/Medium/Bodies/F_MED_Neon_Tight_Suit/Materials/Soft_Perlin_output}
  n4 = VertexNormalWS()
  Multiply(A=@n1.scale, B=Multiply(A=Multiply(A=@n3.R, B=@n4) {constB=20}, B=@n1.WPOStr) {constB=4})
  ParticlePositionWS()
  TextureCoordinate() {uTiling=0.2, vTiling=0.2}
  TextureSample(UVs=TextureCoordinate() {uTiling=0.25, vTiling=0.25}) {samplerType=SAMPLERTYPE_Normal, texture=/FNGameplayCues/Effects/SnowSandInteraction/Textures/SSI_T_WaveNormal}
  VectorParameter() {parameterName=Color, defaultValue=<rgba 0.2333 0.2394 0.25 0>}
 ]
}
