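# Niagara system for snow and sand foot interaction: leaves fading footprint meshes and kicks up particles.
# On valid spawn frames, Footprint places a mesh aligned to ground normal and foot direction, except on ice.
# GPU sprites and snow-only colliding mesh chunks read User.SpawnList, spawning for a brief window, scaled by bone speed and camera distance.
# Time-of-day curves adjust colour and material parameters; smoke and RenderersWithNoParticles emitters are disabled.
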NiagaraSystem /FNGameplayCues/Effects/SnowSandInteraction/Niagara/SSI_NS_InteractionFX_Snow
dependencies {
 usedDataInterfaces [
  &NiagaraDataInterfaceColorCurve
  &NiagaraDataInterfaceCurve
  &NiagaraDataInterfaceCollisionQuery
 ]
 usedDynamicInputs [&self:V4FArray]
 usedModules [
  &self:SystemSpawnScript.SystemScriptSource.SystemScriptGraph.NiagaraNodeAssignment_2.SetVariables_#1
  &self:SystemSpawnScript.SystemScriptSource.SystemScriptGraph.NiagaraNodeAssignment_0.SetVariables_#2
  &self:SystemSpawnScript.SystemScriptSource.SystemScriptGraph.NiagaraNodeAssignment_3.SetVariables_#3
  &SystemState
  &self:DetectSpawning
  &self:SystemSpawnScript.SystemScriptSource.SystemScriptGraph.NiagaraNodeAssignment_1.SetVariables_#4
  &self:ClearArray
 ]
}
emitters {
 Footprint {
  emitterUpdateScript [
   &EmitterState()
   &SpawnPerFrame(
    Spawn Count:NiagaraInt32=&MakeCustomIntFromBool(
     Bool=$System.ValidFrame
     True INT=&MakeCustomIntFromBool(Bool=$User.bIsIce, True INT=0, False INT=1)
     False INT=0
    )
    Use Spawn Probability=false
   )
  ]
  particleSpawnScript [
   &InitializeParticle(
    Lifetime=2.2
    Mesh Scale Mode=::NewEnumerator3
    Mesh Uniform Scale=1.1
    Position=&self:V3FromV4(New Vector 4:Vector4f=$System.SpawnInfo)
    Position Mode=::NewEnumerator1
    Position Offset=<0 0 2>
    UsePositionOffset=true
    Write Material Random=true
   )
   &InitialMeshOrientation(Mesh Orientation Mode=::NewEnumerator7, Use Rotation Vector=true)
   &self:FindFlowRotation(WorldNormal:Vector3f=$User.FootDirection) {module=FindFlowRotation}
   /SetVariables(Particles.MeshOrientation=&MultiplyQuaternion(
     Quaternion A=&FindQuatBetweenTwoVectors(From Vector=<0 0 1>, To Vector=$User.GroundNormal)
     Quaternion B=&MakeQuatFromAxisAngle(Axis=<0 0 1>, AngleInDegrees=&Lerp_Float(A=0, B=-360, Alpha=$Particles.Rotation))
    )) {isSetParameters=true}
   /SetVariables(Particles.bRight:NiagaraBool=$User.bIsRight) {isSetParameters=true}
   /SetVariables(Particles.TintScale:NiagaraFloat=&FN_Select_Float_from_Time_Of_Day(Morning=1, Day=1, Evening=1, Night=1.2)) {isSetParameters=true}
   /SetVariables(Particles.SandFallbackColor:LinearColor=&ColorFromCurve(
     ColorCurve=DataInterface() {propertyValues {
       AlphaCurve {keys [
         <at 0 0.3 RCIM_Cubic>
         <at 0.11 0.3>
         <at 0.125 0.1>
         <at 0.163 0.16>
         <at 0.35 0.2>
         <at 0.41 0.35>
         <at 0.65 0.3>
         <at 0.75 0.16>
         <at 0.833 0.15>
         <at 0.912 0.1>
         <at 0.98 0.17>
         {arriveTangent=6.5, interpMode=RCIM_Cubic, leaveTangent=6.5, time=1, value=0.3}
        ]}
       BlueCurve {keys []}
       GreenCurve {keys []}
       RedCurve {keys []}
      }}
     CurveIndex=Linked() {linkedVariable {name=NPC.FortniteNPC.TimeOfDay_0-1, type=NiagaraFloat}}
    )) {isSetParameters=true}
  ]
  particleUpdateScript [
   &ParticleState()
   /SetVariables(Particles.MarkFlat:NiagaraFloat=&ClampFloat(Float=&FloatFromCurve(
      FloatCurve=DataInterface() {propertyValues {Curve {keys [
          {interpMode=RCIM_Cubic, tangentMode=RCTM_Break, time=0, value=0}
          {arriveTangent=2.7798, interpMode=RCIM_Cubic, leaveTangent=2.7798, tangentMode=RCTM_User, time=1, value=1}
         ]}}}
      Scale Curve=&RandomRangeFloat(Minimum=1, Maximum=2)
     ))) {isSetParameters=true}
   /SetVariables(Particles.BlurMaskScale:NiagaraFloat=&Multiply_Float(
     A=&FloatFromCurve(
      FloatCurve=DataInterface() {propertyValues {Curve {keys [{arriveTangent=6.5413, time=0, value=0}, <at 0.1 1>]}}}
      CurveIndex=$Particles.Age
      Scale Curve=&RandomRangeFloat(Minimum=0.8, Maximum=1)
     )
     B=&FloatFromCurve(FloatCurve=DataInterface() {propertyValues {Curve {keys [
          {arriveTangent=6.5413, interpMode=RCIM_Cubic, tangentMode=RCTM_Break, time=0, value=1}
          {arriveTangent=1.114, interpMode=RCIM_Cubic, tangentMode=RCTM_Break, time=1, value=1.35}
         ]}}})
    )) {isSetParameters=true}
   &DynamicMaterialParameters(
    Index 0 Param 2=$Particles.BlurMaskScale
    Index 0 Param 3=$Particles.MarkFlat
    Index 0 Param 4=&MakeFloatFromVector(Vector=$Particles.Scale, Channel=::NewEnumerator0)
    Index 1 Param 2=&MakeCustomFloatFromBool(Bool=$Particles.bRight, True Float=1, False Float=0)
    Index 1 Param 3=&RandomRangeFloat(Minimum=0, Maximum=5)
    Index 1 Param 4=0.75
    Index 2 Param 1=$Particles.TintScale
    Index 2 Param 2=1.5
    Index 3 Param 1=&MakeFloatFromVector4(
     Vector 4=&self:Vector4FromQuat(QUat:Quat4f=$Particles.MeshOrientation)
     Channel=::NewEnumerator0
    )
    Index 3 Param 2=&MakeFloatFromVector4(
     Vector 4=&self:Vector4FromQuat(QUat:Quat4f=$Particles.MeshOrientation)
     Channel=::NewEnumerator1
    )
    Index 3 Param 3=&MakeFloatFromVector4(
     Vector 4=&self:Vector4FromQuat(QUat:Quat4f=$Particles.MeshOrientation)
     Channel=::NewEnumerator2
    )
    Index 3 Param 4=&MakeFloatFromVector4(
     Vector 4=&self:Vector4FromQuat(QUat:Quat4f=$Particles.MeshOrientation)
     Channel=::NewEnumerator3
    )
    Module.Param0WriteEnabled=true
    Module.Param0WriteEnabled1=true
    Module.Param0WriteEnabled2=true
    Module.Param0WriteEnabled3=true
    Module.Param1WriteEnabled=true
    Module.Param1WriteEnabled1=true
    Module.Param1WriteEnabled2=true
    Module.Param1WriteEnabled3=true
    Module.Param2WriteEnabled=true
    Module.Param2WriteEnabled1=true
    Module.Param2WriteEnabled2=true
    Module.Param2WriteEnabled3=true
    Module.Param3WriteEnabled=true
    Module.Param3WriteEnabled1=true
    Module.Param3WriteEnabled2=true
    Module.Param3WriteEnabled3=true
    Write Parameter Index 1=true
    Write Parameter Index 2=true
    Write Parameter Index 3=true
   )
   &ScaleColor(
    Scale Alpha=&FloatFromCurve(
     FloatCurve=DataInterface() {propertyValues {Curve {keys [
         {tangentMode=RCTM_User, time=0, value=1}
         {arriveTangent=-0.0405, interpMode=RCIM_Cubic, leaveTangent=-0.0405, tangentMode=RCTM_User, time=0.6, value=1}
         {arriveTangent=-0.7677, interpMode=RCIM_Cubic, leaveTangent=-0.7677, tangentMode=RCTM_User, time=1, value=0}
        ]}}}
     Scale Curve=0.5
    )
    ScaleA=true
    ScaleRGB=true
   )
  ]
  rendererClasses [&NiagaraMeshRendererProperties]
  renderers [
   {rendererClass=&NiagaraMeshRendererProperties, rendererIndex=0}
   {rendererClass=&NiagaraMeshRendererProperties, rendererIndex=1}
   {rendererClass=&NiagaraMeshRendererProperties, rendererIndex=2}
  ]
 }
 Fountain {
  simTarget=GPUComputeSim
  emitterSpawnScript [/SetVariables(Emitter.AgeFroSpawn=1000000000) {isSetParameters=true}]
  emitterUpdateScript [
   &EmitterState()
   /SetVariables(Emitter.SpeedFactor:NiagaraFloat=&FloatFromCurve(
     FloatCurve=DataInterface() {propertyValues {Curve {keys [<at 0 0>,<at 200 0.4>,<at 550 1>]}}}
     CurveIndex=&VectorLength(Vector=$User.BoneSpeed)
    )) {isSetParameters=true}
   /SetVariables(Emitter.DistanceFactor:NiagaraFloat=&OneMinusFloat(Float=&ClampFloat(Float=&FN_LOD_Distance_Remap(Start Distance=1000, End Distance=2000, Minimum Output=0, Maximum Output=1)))) {isSetParameters=true}
   &SpawnPerFrame(
    Spawn Count:NiagaraInt32=&MakeCustomIntFromBool(
     Bool=&SetBoolByFloatComparison(A=$Emitter.Age, B=&Add_Float(A=$Emitter.AgeFroSpawn, B=0.1))
     True INT=&MakeIntFromFloat(Float=&Multiply_Float(
       A=&Multiply_Float(A=&Multiply_Float(A=5, B=$Emitter.DistanceFactor), B=$Emitter.SpeedFactor)
       B=&MakeCustomFloatFromBool(Bool=$User.bIsSnow, True Float=1, False Float=0)
      ))
     False INT=0
    )
    Spawn Group=1
    Use Spawn Probability=false
   ) {module=SpawnPerFrame001}
   /SetVariables(Emitter.AgeFroSpawn:NiagaraFloat=&MakeCustomFloatFromBool(
     Bool=$System.ValidFrame
     True Float=$Emitter.Age
     False Float=&MakeCustomFloatFromBool(
      Bool=&SetBoolByFloatComparison(A=$Emitter.Age, B=&Add_Float(A=$Emitter.AgeFroSpawn, B=0.1))
      True Float=1000000000
      False Float=$Emitter.AgeFroSpawn
     )
    )) {isSetParameters=true}
  ]
  particleSpawnScript [
   &self:GetSpawnInfo_01(TargetArray:NiagaraDataInterfaceArrayFloat4=$User.SpawnList) {module=GetSpawnInfo_01}
   &InitializeParticle(
    Color=<rgba 1 0 0.3111 1>
    Lifetime Max=1.75
    Lifetime Min=1.4
    Lifetime Mode=::NewEnumerator1
    Mass Mode=::NewEnumerator3
    Mesh Scale Mode=::NewEnumerator3
    Position=&Add_Vector(A=$Particles.SpawnPosition, B=Linked() {linkedVariable {name=Engine.Owner.Position, type=NiagaraPosition}})
    Position Mode=::NewEnumerator1
    Position Offset=<0 0 -45>
    Sprite Rotation Mode=::NewEnumerator2
    Sprite Size Mode=::NewEnumerator1
    Uniform Sprite Size Max=1
    Uniform Sprite Size Min=1
    UsePositionOffset=true
    Write Material Random=true
   )
   &SphereLocation(
    MaskSpawnGroup=false
    Override Seed=false
    Sphere Radius=8
    Surface Only=false
    UseNonUniformScale=false
    UseOffset=false
    UseSphereOrientation=false
   )
   &AddVelocityInCone(
    Cone Angle=20
    Cone Axis=&NormalizeVector(
     Vector To Normalize=&Add_Vector(
      A=&NormalizeVector(Vector To Normalize=$User.BoneSpeed, Safe Normalize=true, Safe Fallback Threshold=0, Safe Fallback Vector=<0 0 0.001>)
      B=&Add_Vector(A=&Multiply_Vector(A=$User.GroundNormal, B=<0 0 0.3>), B=<0 0 1>)
     )
     Safe Normalize=true
     Safe Fallback Threshold=0
     Safe Fallback Vector=<0 0 0.001>
    )
    Cone Axis Coordinate Space=::World
    Module.Use Velocity Falloff On Cone Axis=false
    Velocity Distribution Along Cone Axis=0.75
    Velocity Strength=&Multiply_Float(
     A=&RandomRangeFloat(Minimum=90, Maximum=300)
     B=&FloatFromCurve(
      FloatCurve=DataInterface() {propertyValues {Curve {keys [
          {interpMode=RCIM_Cubic, leaveTangent=2.3881, tangentMode=RCTM_Break, time=0, value=0}
          {arriveTangent=1.3675, interpMode=RCIM_Cubic, tangentMode=RCTM_Break, time=1, value=1}
         ]}}}
      CurveIndex=$Emitter.SpeedFactor
     )
    )
   )
  ]
  particleUpdateScript [
   &ParticleState()
   &ScaleMeshSize(Scale Factor=&VectorFromFloat(Value=&Lerp_Float(
      A=0.01
      B=0.04
      Alpha=&FloatFromCurve(
       FloatCurve=DataInterface() {propertyValues {Curve {keys [
           {interpMode=RCIM_Cubic, tangentMode=RCTM_Break, time=0, value=0}
           {arriveTangent=3.2169, interpMode=RCIM_Cubic, tangentMode=RCTM_Break, time=1, value=1}
          ]}}}
       CurveIndex=&RandomRangeFloat(Minimum=0, Maximum=1)
       Scale Curve=&FloatFromCurve(
        FloatCurve=DataInterface() {propertyValues {Curve {keys [{tangentMode=RCTM_User, time=0, value=0.5}, {tangentMode=RCTM_User, time=1, value=1}]}}}
        CurveIndex=$Emitter.SpeedFactor
       )
      )
     )))
   &ScaleMeshSize(Scale Factor=&VectorFromFloat(Value=&FloatFromCurve(FloatCurve=DataInterface() {propertyValues {Curve {keys [
          {tangentMode=RCTM_User, time=0, value=1}
          {arriveTangent=-0.0405, interpMode=RCIM_Cubic, leaveTangent=-0.0405, tangentMode=RCTM_User, time=0.2, value=1}
          {arriveTangent=-4.3345, interpMode=RCIM_Cubic, leaveTangent=-4.3345, tangentMode=RCTM_User, time=1, value=0}
         ]}}}))) {module=ScaleMeshSize001}
   &GravityForce()
   &Drag(Drag=0.25, Use Linear Drag=true, Use Rotational Drag=false)
   &ScaleColor(
    Scale Alpha=&FloatFromCurve(FloatCurve=DataInterface() {propertyValues {Curve {keys [<at -0.0046 1.0113 RCIM_Cubic>,<at 1.0011 0.002 RCIM_Cubic>]}}})
    ScaleA=true
    ScaleRGB=true
   )
   &Collision(
    Age Colliding Particles=false
    Friction=155555
    GPU Collision Type=::NewEnumerator2
    Module.Control Rest State Via Surface Normal=false
    Module.EnableMaxCollisionCount=false
    Module.Randomize Collision Normal Vector=false
    Restitution=0
   )
   &SolveForcesAndVelocity(Clamp Velocity=false, Limit Acceleration=false)
   &DynamicMaterialParameters(
    Index 0 Param 1=&RandomRangeFloat(Minimum=0, Maximum=5)
    Index 0 Param 2=&MakeFloatFromVector(Vector=$Particles.Scale, Channel=::NewEnumerator0)
    Index 0 Param 3=150
    Index 0 Param 4=&FNTimeOfDayScalar(Night Scale=0, Morning Scale=0, Day Scale=0.2, Evening Scale=0)
    Module.Param0WriteEnabled=true
    Module.Param1WriteEnabled=true
    Module.Param2WriteEnabled=true
    Module.Param3WriteEnabled=true
   )
  ]
  rendererClasses [&NiagaraMeshRendererProperties]
  renderers [{rendererClass=&NiagaraMeshRendererProperties, rendererIndex=0}]
 }
 RenderersWithNoParticles {
  bEnabled=false
  emitterSpawnScript [/SetVariables(Emitter.Color:LinearColor=<rgba 0.0765 0 1 1>) {isSetParameters=true}]
  emitterUpdateScript [
   &EmitterState()
   /SetVariables(Emitter.SpriteSize:Vector2f=<20 20>) {isSetParameters=true}
   /SetVariables(Emitter.VersionPos:Vector3f=<25 0 25>) {isSetParameters=true}
  ]
  rendererClasses [&NiagaraMeshRendererProperties, &NiagaraSpriteRendererProperties]
  renderers [
   {rendererClass=&NiagaraMeshRendererProperties, rendererIndex=0}
   {rendererClass=&NiagaraSpriteRendererProperties, rendererIndex=1}
   {rendererClass=&NiagaraSpriteRendererProperties, rendererIndex=2}
  ]
 }
 SandSnowParticles {
  simTarget=GPUComputeSim
  emitterSpawnScript [/SetVariables(Emitter.AgeFroSpawn=1000000000) {isSetParameters=true}]
  emitterUpdateScript [
   &EmitterState()
   /SetVariables(Emitter.SpeedFactor:NiagaraFloat=&FloatFromCurve(
     FloatCurve=DataInterface() {propertyValues {Curve {keys [<at 0 0>,<at 200 0.4>,<at 550 1>]}}}
     CurveIndex=&VectorLength(Vector=$User.BoneSpeed)
    )) {isSetParameters=true}
   /SetVariables(Emitter.DistanceFactor:NiagaraFloat=&OneMinusFloat(Float=&ClampFloat(Float=&FN_LOD_Distance_Remap(Start Distance=2000, End Distance=3000, Minimum Output=0, Maximum Output=1)))) {isSetParameters=true}
   &SpawnPerFrame(
    Spawn Count:NiagaraInt32=&MakeCustomIntFromBool(
     Bool=&SetBoolByFloatComparison(A=$Emitter.Age, B=&Add_Float(A=$Emitter.AgeFroSpawn, B=0.1))
     True INT=&MakeIntFromFloat(Float=&Multiply_Float(
       A=&Multiply_Float(A=&Multiply_Float(A=720, B=$Emitter.DistanceFactor), B=$Emitter.SpeedFactor)
       B=&MakeCustomFloatFromBool(Bool=$User.bIsSnow, True Float=0.4, False Float=1)
      ))
     False INT=0
    )
    Use Spawn Probability=false
   )
   /SetVariables(Emitter.AgeFroSpawn:NiagaraFloat=&MakeCustomFloatFromBool(
     Bool=$System.ValidFrame
     True Float=$Emitter.Age
     False Float=&MakeCustomFloatFromBool(
      Bool=&SetBoolByFloatComparison(A=$Emitter.Age, B=&Add_Float(A=$Emitter.AgeFroSpawn, B=0.1))
      True Float=1000000000
      False Float=$Emitter.AgeFroSpawn
     )
    )) {isSetParameters=true}
  ]
  particleSpawnScript [
   &self:GetSpawnInfo(TargetArray:NiagaraDataInterfaceArrayFloat4=$User.SpawnList) {module=GetSpawnInfo}
   &InitializeParticle(
    Color=&ColorFromCurve(
     ColorCurve=DataInterface() {propertyValues {
       AlphaCurve {keys [<at 0 1 RCIM_Cubic>,<at 1 1 RCIM_Cubic>]}
       BlueCurve {keys [
         {arriveTangent=3.675, interpMode=RCIM_Cubic, leaveTangent=3.675, time=0, value=0.35}
         <at 0.0816 0.65>
         <at 0.125 0.4>
         <at 0.225 0.25>
         <at 0.25 0.5>
         <at 0.52 0.22>
         <at 0.5836 0.35>
         <at 0.7286 0.5>
         {arriveTangent=-2.0726, interpMode=RCIM_Cubic, leaveTangent=-2.0726, time=0.8, value=0.35}
         <at 0.8492 0.25>
         <at 0.9 0.7>
         <at 1 0.35>
        ]}
       GreenCurve {keys [
         {arriveTangent=3.675, interpMode=RCIM_Cubic, leaveTangent=3.675, time=0, value=0.35}
         <at 0.0816 0.65>
         <at 0.125 0.4>
         <at 0.225 0.25>
         <at 0.25 0.5>
         <at 0.52 0.22>
         <at 0.5836 0.35>
         <at 0.7286 0.5>
         {arriveTangent=-2.0726, interpMode=RCIM_Cubic, leaveTangent=-2.0726, time=0.8, value=0.35}
         <at 0.8492 0.25>
         <at 0.9 0.7>
         <at 1 0.35>
        ]}
       RedCurve {keys [
         {arriveTangent=3.675, interpMode=RCIM_Cubic, leaveTangent=3.675, time=0, value=0.35}
         <at 0.0816 0.65>
         <at 0.125 0.4>
         <at 0.225 0.25>
         <at 0.25 0.5>
         <at 0.52 0.22>
         <at 0.5836 0.35>
         <at 0.7286 0.5>
         {arriveTangent=-2.0726, interpMode=RCIM_Cubic, leaveTangent=-2.0726, time=0.8, value=0.35}
         <at 0.8492 0.25>
         <at 0.9 0.7>
         <at 1 0.35>
        ]}
      }}
     CurveIndex=Linked() {linkedVariable {name=NPC.FortniteNPC.TimeOfDay_0-1, type=NiagaraFloat}}
    )
    Lifetime Max=1.3
    Lifetime Min=0.7
    Lifetime Mode=::NewEnumerator1
    Mass Mode=::NewEnumerator3
    Position=&Add_Vector(A=$Particles.SpawnPosition, B=Linked() {linkedVariable {name=Engine.Owner.Position, type=NiagaraPosition}})
    Position Mode=::NewEnumerator1
    Position Offset=<0 0 -45>
    Sprite Rotation Mode=::NewEnumerator2
    Sprite Size Mode=::NewEnumerator1
    Uniform Sprite Size Max=15.2825
    Uniform Sprite Size Min=7.0447
    UsePositionOffset=true
    Write Material Random=true
   )
   &SphereLocation(
    MaskSpawnGroup=false
    Non Uniform Scale=<5 1 1>
    Override Seed=false
    Sphere Radius=5
    Surface Only=false
    UseNonUniformScale=true
    UseOffset=false
    UseSphereOrientation=false
   )
   &AddVelocityInCone(
    Cone Angle=20
    Cone Axis=&NormalizeVector(
     Vector To Normalize=&Add_Vector(
      A=&NormalizeVector(Vector To Normalize=$User.BoneSpeed, Safe Normalize=true, Safe Fallback Threshold=0, Safe Fallback Vector=<0 0 0.001>)
      B=&Add_Vector(A=&Multiply_Vector(A=$User.GroundNormal, B=<0 0 0.3>), B=<0 0 1>)
     )
     Safe Normalize=true
     Safe Fallback Threshold=0
     Safe Fallback Vector=<0 0 0.001>
    )
    Cone Axis Coordinate Space=::World
    Module.Use Velocity Falloff On Cone Axis=false
    Velocity Distribution Along Cone Axis=0.75
    Velocity Strength=&Multiply_Float(
     A=&RandomRangeFloat(Minimum=90, Maximum=300)
     B=&FloatFromCurve(
      FloatCurve=DataInterface() {propertyValues {Curve {keys [
          {interpMode=RCIM_Cubic, leaveTangent=2.3881, tangentMode=RCTM_Break, time=0, value=0}
          {arriveTangent=1.3675, interpMode=RCIM_Cubic, tangentMode=RCTM_Break, time=1, value=1}
         ]}}}
      CurveIndex=$Emitter.SpeedFactor
     )
    )
   )
   /SetVariables(
    Particles.Lifetime=&Multiply_Float(
     A=1
     B=&MakeCustomFloatFromBool(
      Bool=/Niagara/DynamicInputs/Helpers/MaskBoolBySpawnGroup(Masked Bool:NiagaraBool=false, Passthrough Bool:NiagaraBool=true, Spawn Group Mask:NiagaraInt32=0)
      True Float=&Multiply_Float(A=$Particles.Lifetime, B=0.5)
      False Float=$Particles.Lifetime
     )
    )
    Particles.VTG:NiagaraInt32=&SpawnGroup()
   ) {isSetParameters=true}
  ]
  particleUpdateScript [
   &ParticleState()
   &LinearForce()
   &self:AlignVelocityWithDF(B=1, B001=-0.35, B002=2, B003=0.98, New bool:NiagaraBool=&SetBoolByFloatComparison(A=$Particles.Age, B=0.2)) {module=AlignVelocityWithDF}
   &GravityForce(Gravity=&Multiply_VectorByFloat(
     Vector=<0 0 -980>
     Float=&FloatFromCurve(FloatCurve=DataInterface() {propertyValues {Curve {keys [
          {tangentMode=RCTM_User, time=0, value=1}
          {arriveTangent=-0.0405, interpMode=RCIM_Cubic, leaveTangent=-0.0405, tangentMode=RCTM_User, time=0.2931, value=1}
          {arriveTangent=-3.8325, interpMode=RCIM_Cubic, leaveTangent=-3.8325, tangentMode=RCTM_User, time=0.785, value=0}
         ]}}})
    ))
   &self:SnapToSurface() {module=SnapToSurface}
   &Drag(
    Drag=&MakeCustomFloatFromBool(
     Bool=$Particles.Result
     True Float=&Lerp_Float(
      A=0
      B=10
      Alpha=&FloatFromCurve(
       FloatCurve=DataInterface() {propertyValues {Curve {keys [
           {arriveTangent=0.2486, interpMode=RCIM_Cubic, leaveTangent=0.2486, tangentMode=RCTM_User, time=0, value=0}
           {arriveTangent=1.1682, interpMode=RCIM_Cubic, leaveTangent=1.1682, tangentMode=RCTM_User, time=0.8534, value=0.4625}
           {arriveTangent=5.2078, interpMode=RCIM_Cubic, leaveTangent=5.2078, tangentMode=RCTM_User, time=1, value=1}
          ]}}}
       CurveIndex=&Abs_Float(Float:NiagaraFloat=$Particles.GravF)
      )
     )
     False Float=0
    )
    Use Linear Drag=true
    Use Rotational Drag=false
   )
   &ScaleColor(
    Scale Alpha=&FloatFromCurve(FloatCurve=DataInterface() {propertyValues {Curve {keys [
         {tangentMode=RCTM_User, time=0, value=0}
         <at 0.2 1>
         {arriveTangent=-0.0405, interpMode=RCIM_Cubic, leaveTangent=-0.0405, tangentMode=RCTM_User, time=0.4027, value=1}
         {arriveTangent=-1.9583, interpMode=RCIM_Cubic, leaveTangent=-1.9583, tangentMode=RCTM_User, time=1, value=0}
        ]}}})
    ScaleA=true
    ScaleRGB=true
   )
   &SolveForcesAndVelocity(Clamp Velocity=false, Limit Acceleration=false)
   &ScaleSpriteSizeBySpeed(Max Scale Factor=<1 10>, Min Scale Factor=<1 2>)
   &DynamicMaterialParameters(
    Index 0 Param 1=&RandomRangeFloat(Minimum=0, Maximum=10)
    Index 0 Param 2=&MakeFloatFromLinearColor(LinearColor=$Particles.Color, Channel:ENiagaraLinearColor_Channels=::NewEnumerator0)
    Index 0 Param 3=&MakeFloatFromLinearColor(LinearColor=$Particles.Color, Channel:ENiagaraLinearColor_Channels=::NewEnumerator1)
    Index 0 Param 4=&MakeFloatFromLinearColor(LinearColor=$Particles.Color, Channel:ENiagaraLinearColor_Channels=::NewEnumerator2)
    Module.Param0WriteEnabled=true
    Module.Param1WriteEnabled=true
    Module.Param2WriteEnabled=true
    Module.Param3WriteEnabled=true
   )
  ]
  rendererClasses [&NiagaraSpriteRendererProperties]
  renderers [{rendererClass=&NiagaraSpriteRendererProperties}]
 }
 SandSnowSmoke {
  bEnabled=false
  simTarget=GPUComputeSim
  emitterSpawnScript [/SetVariables(Emitter.Agegg=1000000000) {isSetParameters=true}]
  emitterUpdateScript [
   &EmitterState()
   /SetVariables(Emitter.SpeedFactor:NiagaraFloat=&FloatFromCurve(
     FloatCurve=DataInterface() {propertyValues {Curve {keys [<at 0 0>,<at 200 0.4>,<at 550 1>]}}}
     CurveIndex=&VectorLength(Vector=$User.BoneSpeed)
    )) {isSetParameters=true}
   /SetVariables(Emitter.DistanceFactor:NiagaraFloat=&OneMinusFloat(Float=&ClampFloat(Float=&FN_LOD_Distance_Remap(Start Distance=1500, End Distance=1600, Minimum Output=0, Maximum Output=1)))) {isSetParameters=true}
   &SpawnPerFrame(
    Spawn Count:NiagaraInt32=&MakeCustomIntFromBool(
     Bool=&SetBoolByFloatComparison(A=$Emitter.Age, B=&Add_Float(A=$Emitter.Agegg, B=0.1))
     True INT=&MakeIntFromFloat(Float=&Multiply_Float(A=1, B=$Emitter.DistanceFactor))
     False INT=0
    )
    Use Spawn Probability=false
   )
   /SetVariables(Emitter.Agegg:NiagaraFloat=&MakeCustomFloatFromBool(
     Bool=$System.ValidFrame
     True Float=$Emitter.Age
     False Float=&MakeCustomFloatFromBool(
      Bool=&SetBoolByFloatComparison(A=$Emitter.Age, B=&Add_Float(A=$Emitter.Agegg, B=0.1))
      True Float=1000000000
      False Float=$Emitter.Agegg
     )
    )) {isSetParameters=true}
  ]
  particleSpawnScript [
   &self:GetSpawnInfo(TargetArray:NiagaraDataInterfaceArrayFloat4=$User.SpawnList) {module=GetSpawnInfo}
   &InitializeParticle(
    Color=&MakeCustomLinearColorFromBool(
     Bool=$User.bIsSnow
     True Color=&MultiplyLinearColor(
      Linear Color=&ColorFromCurve(
       ColorCurve=DataInterface() {propertyValues {
         AlphaCurve {keys [
           {arriveTangent=0.12, interpMode=RCIM_Cubic, leaveTangent=0.12, time=0, value=0.31}
           <at 0.5 0.37>
           {arriveTangent=-0.12, interpMode=RCIM_Cubic, leaveTangent=-0.12, time=1, value=0.31}
          ]}
         BlueCurve {keys [<at 0 1 RCIM_Cubic>,<at 0.12 1>,<at 0.5 1 RCIM_Cubic>,<at 0.9 1>,<at 1 1>]}
         GreenCurve {keys [<at 0 1 RCIM_Cubic>,<at 0.12 1>,<at 0.5 1 RCIM_Cubic>,<at 0.9 1>,<at 1 1>]}
         RedCurve {keys [<at 0 1 RCIM_Cubic>,<at 0.12 1>,<at 0.5 1 RCIM_Cubic>,<at 0.9 1>,<at 1 1>]}
        }}
       CurveIndex=Linked() {linkedVariable {name=NPC.FortniteNPC.TimeOfDay_0-1, type=NiagaraFloat}}
      )
      Scale Factor=&MakeVector4(
       X=1
       Y=1
       Z=1
       W=&FloatFromCurve(
        FloatCurve=DataInterface() {propertyValues {Curve {keys [
            {arriveTangent=1.7335, interpMode=RCIM_Cubic, leaveTangent=1.7335, tangentMode=RCTM_User, time=0, value=0}
            {arriveTangent=0.9995, interpMode=RCIM_Cubic, leaveTangent=0.9995, tangentMode=RCTM_User, time=0.3549, value=0.55}
            {arriveTangent=-0.0405, interpMode=RCIM_Cubic, leaveTangent=-0.0405, tangentMode=RCTM_User, time=1, value=1}
           ]}}}
        CurveIndex=$Emitter.SpeedFactor
       )
      )
     )
     False Color=&MultiplyLinearColor(
      Linear Color=&ColorFromCurve(
       ColorCurve=DataInterface() {propertyValues {
         AlphaCurve {keys [
           {arriveTangent=0.12, interpMode=RCIM_Cubic, leaveTangent=0.12, time=0, value=0.31}
           <at 0.5 0.37>
           {arriveTangent=-0.12, interpMode=RCIM_Cubic, leaveTangent=-0.12, time=1, value=0.31}
          ]}
         BlueCurve {keys [
           {arriveTangent=0.056, interpMode=RCIM_Cubic, leaveTangent=0.056, time=0, value=0.1673}
           <at 0.12 0.174>
           <at 0.5 0.1656 RCIM_Cubic>
           <at 0.9 0.174>
           <at 1 0.1673>
          ]}
         GreenCurve {keys [
           {arriveTangent=-0.1262, interpMode=RCIM_Cubic, leaveTangent=-0.1262, time=0, value=0.2277}
           <at 0.12 0.2126>
           <at 0.5 0.2626 RCIM_Cubic>
           <at 0.9 0.2126>
           <at 1 0.2277>
          ]}
         RedCurve {keys [
           {arriveTangent=-0.475, interpMode=RCIM_Cubic, leaveTangent=-0.475, time=0, value=0.409}
           <at 0.12 0.352>
           <at 0.5 0.508 RCIM_Cubic>
           <at 0.9 0.352>
           <at 1 0.409>
          ]}
        }}
       CurveIndex=Linked() {linkedVariable {name=NPC.FortniteNPC.TimeOfDay_0-1, type=NiagaraFloat}}
      )
      Scale Factor=&MakeVector4(
       X=1
       Y=1
       Z=1
       W=&FloatFromCurve(
        FloatCurve=DataInterface() {propertyValues {Curve {keys [
            {arriveTangent=1.7335, interpMode=RCIM_Cubic, leaveTangent=1.7335, tangentMode=RCTM_User, time=0, value=0}
            {arriveTangent=0.9995, interpMode=RCIM_Cubic, leaveTangent=0.9995, tangentMode=RCTM_User, time=0.3549, value=0.55}
            {arriveTangent=-0.0405, interpMode=RCIM_Cubic, leaveTangent=-0.0405, tangentMode=RCTM_User, time=1, value=1}
           ]}}}
        CurveIndex=$Emitter.SpeedFactor
       )
      )
     )
    )
    Lifetime Max=1
    Lifetime Min=0.8
    Lifetime Mode=::NewEnumerator1
    Position:Vector3f=$Particles.SpawnPosition
    Position Mode=::NewEnumerator1
    Position Offset=<0 0 -6>
    Sprite Size Mode=::NewEnumerator1
    Uniform Sprite Size Max=180
    Uniform Sprite Size Min=140
    UsePositionOffset=true
    Write Material Random=true
   )
   &AddVelocityInCone(
    Cone Angle=15
    Cone Axis=&NormalizeVector(
     Vector To Normalize=&Add_Vector(
      A=&NormalizeVector(Vector To Normalize=$User.BoneSpeed, Safe Normalize=true, Safe Fallback Threshold=0, Safe Fallback Vector=<0 0 0.001>)
      B=&Add_Vector(A=&Multiply_Vector(A=$User.GroundNormal, B=<0 0 0>), B=<0 0 0.4>)
     )
     Safe Normalize=true
     Safe Fallback Threshold=0
     Safe Fallback Vector=<0 0 0.001>
    )
    Cone Axis Coordinate Space=::World
    Module.Use Velocity Falloff On Cone Axis=false
    Velocity Distribution Along Cone Axis=0.75
    Velocity Strength=&Multiply_Float(
     A=200
     B=&FloatFromCurve(
      FloatCurve=DataInterface() {propertyValues {Curve {keys [
          {arriveTangent=2.4173, interpMode=RCIM_Cubic, leaveTangent=2.4173, tangentMode=RCTM_User, time=0, value=0}
          {arriveTangent=0.9995, interpMode=RCIM_Cubic, leaveTangent=0.9995, tangentMode=RCTM_User, time=0.331, value=0.5492}
          {arriveTangent=0.341, interpMode=RCIM_Cubic, leaveTangent=0.341, tangentMode=RCTM_User, time=1, value=1}
         ]}}}
      CurveIndex=$Emitter.SpeedFactor
     )
    )
   )
  ]
  particleUpdateScript [
   &Drag(Drag=1.8907, Use Linear Drag=true, Use Rotational Drag=false)
   &GravityForce(Gravity=<0 0 -197.7558>)
   &SolveForcesAndVelocity(Clamp Velocity=false, Limit Acceleration=false)
   &ParticleState()
   &ScaleColor(
    Scale Alpha=&FloatFromCurve(
     FloatCurve=DataInterface() {propertyValues {Curve {keys [
         {interpMode=RCIM_Cubic, leaveTangent=9.709, tangentMode=RCTM_Break, time=0, value=0}
         {interpMode=RCIM_Cubic, tangentMode=RCTM_Break, time=0.2, value=1}
         {arriveTangent=-0.0405, interpMode=RCIM_Cubic, leaveTangent=-0.0405, tangentMode=RCTM_User, time=0.4027, value=1}
         {arriveTangent=-1.9583, interpMode=RCIM_Cubic, leaveTangent=-1.9583, tangentMode=RCTM_User, time=1, value=0}
        ]}}}
     Scale Curve=0.5
    )
    ScaleA=true
    ScaleRGB=true
   )
   &DynamicMaterialParameters(
    Index 0 Param 1=&Multiply_Float(A=$Particles.NormalizedAge, B=0.6)
    Index 2 Param 1=1
    Index 2 Param 2=&FloatFromCurve(
     FloatCurve=DataInterface() {propertyValues {Curve {keys [
         {interpMode=RCIM_Cubic, tangentMode=RCTM_User, time=0, value=0}
         {arriveTangent=1.982, interpMode=RCIM_Cubic, leaveTangent=1.982, tangentMode=RCTM_User, time=1, value=1}
        ]}}}
     Scale Curve=0.2
    )
    Index 2 Param 3=&FloatFromCurve(FloatCurve=DataInterface() {propertyValues {Curve {keys [
         {interpMode=RCIM_Cubic, tangentMode=RCTM_User, time=0, value=0}
         {arriveTangent=1.982, interpMode=RCIM_Cubic, leaveTangent=1.982, tangentMode=RCTM_User, time=1, value=1}
        ]}}})
    Index 2 Param 4=1
    Module.Param0WriteEnabled=true
    Module.Param0WriteEnabled2=true
    Module.Param1WriteEnabled=true
    Module.Param1WriteEnabled2=true
    Module.Param2WriteEnabled=true
    Module.Param2WriteEnabled2=true
    Module.Param3WriteEnabled=true
    Module.Param3WriteEnabled2=true
    Write Parameter Index 2=true
   )
   &ScaleSpriteSize(Uniform Curve Sprite Scale=DataInterface() {propertyValues {Curve {keys [
        {arriveTangent=2.0355, interpMode=RCIM_Cubic, leaveTangent=2.0355, tangentMode=RCTM_User, time=0, value=0}
        {arriveTangent=2.153, interpMode=RCIM_Cubic, leaveTangent=2.153, tangentMode=RCTM_User, time=1, value=1}
       ]}}})
   &ScaleSpriteSize(
    Uniform Curve Index=$Emitter.SpeedFactor
    Uniform Curve Sprite Scale=DataInterface() {propertyValues {Curve {keys [
        {arriveTangent=4.1785, interpMode=RCIM_Cubic, leaveTangent=4.1785, tangentMode=RCTM_User, time=0, value=0}
        {arriveTangent=0.9995, interpMode=RCIM_Cubic, leaveTangent=0.9995, tangentMode=RCTM_User, time=0.2223, value=0.6825}
        {arriveTangent=-0.0405, interpMode=RCIM_Cubic, leaveTangent=-0.0405, tangentMode=RCTM_User, time=1, value=1}
       ]}}}
   ) {module=ScaleSpriteSize001}
  ]
  rendererClasses [&NiagaraSpriteRendererProperties]
  renderers [{rendererClass=&NiagaraSpriteRendererProperties}]
 }
}
summary {
 systemName=SSI_NS_InteractionFX_Snow
 emitters [
  {emitterName=Footprint, rendererClasses [&NiagaraMeshRendererProperties]}
  {emitterName=SandSnowParticles, simTarget=GPUComputeSim, rendererClasses [&NiagaraSpriteRendererProperties]}
  {bEnabled=false, emitterName=SandSnowSmoke, simTarget=GPUComputeSim, rendererClasses [&NiagaraSpriteRendererProperties]}
  {
   bEnabled=false
   emitterName=RenderersWithNoParticles
   rendererClasses [&NiagaraMeshRendererProperties, &NiagaraSpriteRendererProperties]
  }
  {emitterName=Fountain, simTarget=GPUComputeSim, rendererClasses [&NiagaraMeshRendererProperties]}
 ]
 userVariables [
  User.bIsIce {type=NiagaraBool}
  User.bIsRight {type=NiagaraBool}
  User.bIsSnow {type=NiagaraBool}
  User.BoneSpeed {type=Vector3f, description="", defaultValue {struct=Vector3f, value=<0 0 0>}}
  User.FootDirection {type=Vector3f, description="", defaultValue {struct=Vector3f, value=<0 0 0>}}
  User.GroundNormal {type=Vector3f, description="", defaultValue {struct=Vector3f, value=<0 0 0>}}
  User.SpawnList {
   type=NiagaraDataInterfaceArrayFloat4
   description=""
   defaultValue {
    struct=NiagaraExt_VariableValue_DataInterface
    value {
     dataInterface=&self:NiagaraDataInterfaceArrayFloat4_0
     dataInterfaceClass=/Script/Niagara.NiagaraDataInterfaceArrayFloat4
    }
   }
  }
  User.TraceEndPos {type=Vector3f, description="", defaultValue {struct=Vector3f, value=<0 0 0>}}
  User.TraceStartPos {type=Vector3f, description="", defaultValue {struct=Vector3f, value=<0 0 0>}}
 ]
}
